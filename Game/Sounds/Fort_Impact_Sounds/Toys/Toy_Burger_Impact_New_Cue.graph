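# Sound cue for impacts of the burger toy, layering two randomized sets of impact sounds on each play.
# A mixer combines two Random nodes at equal volume, one picking from three Toy_Burger_Impact variations and the other from three Burger_Impact variations, all equally weighted.
# It uses the shared toy-hit attenuation and is routed to the Fort_Music_Emote sound class.
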SoundCue /Game/Sounds/Fort_Impact_Sounds/Toys/Toy_Burger_Impact_New_Cue
root=Mixer_2
nodes [
 {id=Mixer_2, children [Random_1,Random_0], properties {inputVolume [1,1]}}
 {id=Random_0, children [WavePlayer_1,WavePlayer_2,WavePlayer_5], properties {weights [1,1,1]}}
 {id=Random_1, children [WavePlayer_0,WavePlayer_3,WavePlayer_4], properties {weights [1,1,1]}}
 {id=WavePlayer_0, properties {soundWaveAssetPtr=&/Toy_Burger_Impact_01}}
 {id=WavePlayer_1, properties {soundWaveAssetPtr=&/Burger_Impact_01}}
 {id=WavePlayer_2, properties {soundWaveAssetPtr=&/Burger_Impact_02}}
 {id=WavePlayer_3, properties {soundWaveAssetPtr=&/Toy_Burger_Impact_02}}
 {id=WavePlayer_4, properties {soundWaveAssetPtr=&/Toy_Burger_Impact_03}}
 {id=WavePlayer_5, properties {soundWaveAssetPtr=&/Burger_Impact_03}}
]
properties {
 attenuationSettings=/Game/Sounds/Attenuation/Fort_ToyHit_Att
 maxDistance=2800
 soundClassObject=&Fort_Music_Emote
 volumeMultiplier=0.8
}
